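# Parent material for the Slurp Proto shoe cosmetic skin, layering the 2023 character Slurp effect onto the standard character material.
# Character PreFX and PostFX functions bracket MF_Slurp_2023, using material attributes and an expression-driven shading model.
# Channel-mask parameters select the Slurp, bubble and AO masks, shared through named reroutes.
# The Slurp normal is flattened by a placeholder-named scalar before PostFX; a power-based mask multiply appears unconnected.
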Material /CosmeticShoes/Assets/FORT_Shoes/Shoe_ChoreNoodlePebble/Skins/Slurp_Proto/Materials/M_Shoe_SlurpProto_Parent
graph {
 properties {
  bUseMaterialAttributes=true
  bUsedWithClothing=true
  bUsedWithMeshDeformer=true
  bUsedWithMorphTargets=true
  bUsedWithNanite=true
  bUsedWithNiagaraMeshParticles=true
  bUsedWithSkeletalMesh=true
  pixelDepthOffsetMode=PDOM_Legacy
  shadingModel=MSM_FromMaterialExpression
  usedShadingModels=Default Lit | Subsurface | Clear Coat | Subsurface Profile | Cloth
 }
 nodes [
  n1 = GetMaterialAttributes(None=MaterialFunctionCall!(
    SlurpMask=NamedRerouteUsage()
    AO=NamedRerouteUsage()
    BubblesMask=NamedRerouteUsage()
    SlurpNormalDistortionMask
    Pre Skinned Position
    Material=NamedRerouteUsage()
   ) {materialFunction=/Game/Global/Materials/Character_Parent_2023/Slurp_Parent_2023/MF_Slurp_2023}.Result) {attributeGetTypes [#1]}
  n2 = ScalarParameter() {parameterName=fasfa}
  n3 = MaterialFunctionCall() {materialFunction=&MF_CharacterParent_2023_PreFX}
  MaterialFunctionCall(
   Material=SetMaterialAttributes!(
    MaterialAttributes=@n1.MaterialAttributes
    Normal=MaterialFunctionCall(Normal=@n1.Normal, Flatness=@n2) {materialFunction=&FlattenNormal}.Result
   ) {attributeSetTypes [#1]}
   M=@n3.M
   Pre Skinned Local Position=@n3.M
   Distance Blend=@n3.M
   Fresnel=@n3.M
  ) {materialFunction=&MF_CharacterParent_2023_PostFX}
  n4 = ComponentMask(None=NamedRerouteUsage()) {r=true}
  n5 = ScalarParameter() {defaultValue=3, parameterName=azg}
  Multiply(A=@n2, B=OneMinus(None=Power(Base=@n4, Exp=@n5)))
  NamedRerouteDeclaration(None=@n3.Tech Art Mask) {name=InTechArtMask, nodeColor=<rgba 1 0.9609 0 1>}
  NamedRerouteDeclaration(None=@n3.M) {name=InMask, nodeColor=<rgba 1 0.7734 0 1>}
  NamedRerouteDeclaration(None=ChannelMaskParameter(None=NamedRerouteUsage()) {group=Slurp, parameterName=SlurpMask}.R) {name=SlurpMask, nodeColor=<rgba 0 1 0.3203 1>}
  NamedRerouteDeclaration(None=ChannelMaskParameter(None=NamedRerouteUsage()) {group=Slurp, parameterName=SlurpBubblesMask}.R) {name=WPOMask, nodeColor=<rgba 0 1 0.8594 1>}
  NamedRerouteDeclaration(None=ChannelMaskParameter(None=NamedRerouteUsage()) {group=Slurp, parameterName=SlurpBubblesMask}.R) {name=BubbleMask, nodeColor=<rgba 0.5938 0 1 1>}
  NamedRerouteDeclaration(None=@n3.Result) {name=InMaterial, nodeColor=<rgba 1 0 0.9609 1>}
  NamedRerouteDeclaration(None=ChannelMaskParameter(None=NamedRerouteUsage()) {group=Slurp, parameterName=AOChannel}.R) {name=AO, nodeColor=<rgba 1 0 0.1875 1>}
 ]
}
